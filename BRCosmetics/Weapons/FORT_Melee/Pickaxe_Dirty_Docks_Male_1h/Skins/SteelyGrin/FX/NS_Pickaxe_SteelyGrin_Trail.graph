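# Niagara swing trail effect for the SteelyGrin pickaxe skin, built from three anim-trail ribbon emitters.
# ColorTrail is a brightly boosted translucent band, SharpEdge a narrower tapering edge, and StarTrail a thin yellow ribbon that widens quickly.
# Each spawns short-lived particles between bone or socket pairs, multiplying initial color by time-of-day tints.
# Curves fade alpha over particle life, while ColorTrail and SharpEdge also fade by system loop age.
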NiagaraSystem /BRCosmetics/Weapons/FORT_Melee/Pickaxe_Dirty_Docks_Male_1h/Skins/SteelyGrin/FX/NS_Pickaxe_SteelyGrin_Trail
dependencies {
 usedDataInterfaces [&NiagaraDataInterfaceColorCurve, &NiagaraDataInterfaceCurve, &NiagaraDataInterfaceSocketReader]
 usedDynamicInputs []
 usedModules [
  &self:SystemSpawnScript.SystemScriptSource.SystemScriptGraph.NiagaraNodeAssignment_1.SetVariables_#1
  &SystemState
 ]
}
emitters {
 ColorTrail {
  emitterSpawnScript [&FN_RibbonIDOffsetUtility()]
  emitterUpdateScript [
   &EmitterState()
   &FN_SpawnAnimTrail()
   /SetVariables(Emitter.Linear Color:LinearColor=&RandomRangeLinearColor(
     Minimum=&ColorFromCurve(
      ColorCurve=DataInterface() {propertyValues {
        AlphaCurve {keys [<at 0 1 RCIM_Cubic>]}
        BlueCurve {keys [<at 0 1 RCIM_Cubic>,<at 1 1 RCIM_Cubic>]}
        GreenCurve {keys [<at 0 0 RCIM_Cubic>,<at 1 0.8991 RCIM_Cubic>]}
        RedCurve {keys [<at 0 0.0734 RCIM_Cubic>,<at 1 0 RCIM_Cubic>]}
       }}
      CurveIndex=$Emitter.Age
     )
     Maximum=&ColorFromCurve(
      ColorCurve=DataInterface() {propertyValues {
        AlphaCurve {keys [<at 0 1 RCIM_Cubic>]}
        BlueCurve {keys [<at 0 0.047 RCIM_Cubic>,<at 1 1 RCIM_Cubic>]}
        GreenCurve {keys [<at 0 0 RCIM_Cubic>,<at 1 0 RCIM_Cubic>]}
        RedCurve {keys [<at 0 1 RCIM_Cubic>,<at 1 0.4538 RCIM_Cubic>]}
       }}
      CurveIndex=$Emitter.Age
     )
    )) {enabled=false, isSetParameters=true}
  ]
  particleSpawnScript [
   &FN_InitializeAnimTrailBoneIndecies(First Bone or Socket Index=1, Second Bone or Socket Index=0)
   &FN_InitializeAnimTrailBoneIndecies(AnimTrail Selection=1, First Bone or Socket Index=1) {enabled=false, module=FN_InitializeAnimTrailBoneIndecies001}
   &FN_InitializeAnimTrail(Color=<rgba 1 1 1 0.13>, Lifetime=0.2)
   &Color(Color=&MultiplyLinearColor(
     Linear Color=$Particles.Initial.Color
     Scale Factor=&MakeVector4_FromLinearColor(LinearColor=&FN_Select_Color_from_Time_Of_Day(Use Manual Time of Day Thresholds=true, Morning=<rgba 1 1 1 0.5>, Day=<rgba 1 1 1 0.9>, Evening=<rgba 1 1 1 0.75>, Night=<rgba 1 1 1 0.5>))
    )) {module=Color001}
   &AddVelocity(Velocity=<0 400 0>) {enabled=false}
  ]
  particleUpdateScript [
   &ScaleColor(
    Scale Alpha=&FloatFromCurve(FloatCurve=DataInterface() {propertyValues {Curve {keys [<at 0.15 1>,<at 0.2 0>]}}})
    Scale RGB=&VectorFromFloat(Value=150)
    ScaleA=true
    ScaleRGB=true
   )
   &ScaleColor(
    Scale Alpha=&FloatFromCurve(FloatCurve=DataInterface() {propertyValues {Curve {keys [<at 0.12 1>,<at 0.15 0>]}}}, CurveIndex=$System.NormalizedLoopAge)
    ScaleA=true
    ScaleRGB=true
   ) {module=ScaleColor002}
   &SolveForcesAndVelocity(Clamp Velocity=false, Limit Acceleration=false)
   &FN_UpdateAnimTrail() {module=FN_UpdateAnimTrail001}
  ]
  rendererClasses [&NiagaraRibbonRendererProperties]
  renderers [
   {rendererClass=&NiagaraRibbonRendererProperties}
   {rendererClass=&NiagaraRibbonRendererProperties, rendererIndex=1}
  ]
 }
 SharpEdge {
  emitterSpawnScript [&FN_RibbonIDOffsetUtility()]
  emitterUpdateScript [
   &EmitterState()
   &FN_SpawnAnimTrail(AnimTrailWidthMode=::FromSecond) {module=FN_SpawnAnimTrail001}
  ]
  particleSpawnScript [
   &FN_InitializeAnimTrailBoneIndecies(First Bone or Socket Index=1, Second Bone or Socket Index=0)
   &FN_InitializeAnimTrail(AnimTrailWidth=0.25, Color=<rgba 1 1 1 0.5>, Lifetime=0.2) {module=FN_InitializeAnimTrail001}
   &Color(Color=&MultiplyLinearColor(
     Linear Color=$Particles.Initial.Color
     Scale Factor=&MakeVector4_FromLinearColor(LinearColor=&FN_Select_Color_from_Time_Of_Day(Use Manual Time of Day Thresholds=true, Morning=<rgba 1 1 1 0.5>, Day=<rgba 1 1 1 0.9>, Evening=<rgba 1 1 1 0.75>, Night=<rgba 1 1 1 0.5>))
    )) {module=Color001}
   &AddVelocity(Velocity=<0 400 0>) {enabled=false}
  ]
  particleUpdateScript [
   &SolveForcesAndVelocity(Clamp Velocity=false, Limit Acceleration=false)
   &ScaleColor(
    Linear Color Curve=DataInterface() {propertyValues {
      AlphaCurve {keys [<at 0.1868 1>,<at 0.2847 0>]}
      BlueCurve {keys [<at 0 0.4948>,<at 0.5969 0.5208>]}
      GreenCurve {keys [<at 0 0.4948>,<at 0.5969 0.5208>]}
      RedCurve {keys [<at 0 0.4948>,<at 0.5969 0.5208>]}
     }}
    Scale Mode=::NewEnumerator2
   )
   &ScaleColor(
    Scale Alpha=&FloatFromCurve(FloatCurve=DataInterface() {propertyValues {Curve {keys [<at 0.12 1>,<at 0.15 0>]}}}, CurveIndex=$System.NormalizedLoopAge)
    ScaleA=true
    ScaleRGB=true
   ) {module=ScaleColor001}
   &FN_UpdateAnimTrail()
   &ScaleRibbonWidth(Ribbon Width Scale=&FloatFromCurve(FloatCurve=DataInterface() {propertyValues {Curve {keys [
         {arriveTangent=0.3072, interpMode=RCIM_Cubic, leaveTangent=0.3072, tangentMode=RCTM_User, time=0, value=1}
         {arriveTangent=-1.4466, interpMode=RCIM_Cubic, leaveTangent=-1.4466, tangentMode=RCTM_User, time=1, value=0}
        ]}}}))
  ]
  rendererClasses [&NiagaraRibbonRendererProperties]
  renderers [
   {rendererClass=&NiagaraRibbonRendererProperties}
   {rendererClass=&NiagaraRibbonRendererProperties, rendererIndex=1}
  ]
 }
 StarTrail {
  emitterSpawnScript [&FN_RibbonIDOffsetUtility()]
  emitterUpdateScript [
   &EmitterState()
   &FN_SpawnAnimTrail(AnimTrailWidthMode=::FromSecond)
  ]
  particleSpawnScript [
   &FN_InitializeAnimTrailBoneIndecies()
   &FN_InitializeAnimTrailBoneIndecies(AnimTrail Selection=1, First Bone or Socket Index=1) {enabled=false, module=FN_InitializeAnimTrailBoneIndecies001}
   &FN_InitializeAnimTrail(AnimTrailWidth=0.15, Color=<rgba 1 0.9771 0 0.3>, Lifetime=0.2)
   &Color(Color=&MultiplyLinearColor(
     Linear Color=$Particles.Initial.Color
     Scale Factor=&MakeVector4_FromLinearColor(LinearColor=&FN_Select_Color_from_Time_Of_Day(Morning=<rgba 0.6719 0.6019 0.6097 1>, Day=<rgba 1 1 1 0.9>, Evening=<rgba 1 0.6413 0.7956 1>, Night=<rgba 0.4668 0.9064 2 0.75>))
    )) {module=Color001}
   &AddVelocity(Velocity=<0 400 0>) {enabled=false}
  ]
  particleUpdateScript [
   &ScaleColor(
    Scale Alpha=&FloatFromCurve(FloatCurve=DataInterface() {propertyValues {Curve {keys [<at 0.09 1>,<at 0.12 0>]}}})
    ScaleA=true
    ScaleRGB=true
   )
   &SolveForcesAndVelocity(Clamp Velocity=false, Limit Acceleration=false)
   &FN_UpdateAnimTrail() {module=FN_UpdateAnimTrail001}
   &ScaleRibbonWidth(Ribbon Width Scale=&FloatFromCurve(FloatCurve=DataInterface() {propertyValues {Curve {keys [
         {arriveTangent=9.4618, interpMode=RCIM_Cubic, leaveTangent=9.4618, tangentMode=RCTM_User, time=0, value=0}
         {arriveTangent=0.85, interpMode=RCIM_Cubic, leaveTangent=0.85, time=0.2, value=1}
         <at 1 0.85>
        ]}}}))
  ]
  rendererClasses [&NiagaraRibbonRendererProperties]
  renderers [
   {rendererClass=&NiagaraRibbonRendererProperties}
   {rendererClass=&NiagaraRibbonRendererProperties, rendererIndex=1}
  ]
 }
}
summary {
 systemName=NS_Pickaxe_SteelyGrin_Trail
 emitters [
  {emitterName=ColorTrail, rendererClasses [&NiagaraRibbonRendererProperties]}
  {emitterName=StarTrail, rendererClasses [&NiagaraRibbonRendererProperties]}
  {emitterName=SharpEdge, rendererClasses [&NiagaraRibbonRendererProperties]}
 ]
 userVariables [
  User.bIsFrontEnd {type=NiagaraBool}
  User.bIsFrontEndPreview {type=NiagaraBool}
  User.PhysicalSurface {type=NiagaraInt32}
  User.RibbonIDOffset {type=NiagaraInt32}
  User.SkeletalMesh {
   type=NiagaraDataInterfaceSkeletalMesh
   defaultValue {
    struct=NiagaraExt_VariableValue_DataInterface
    value {dataInterface=&self:NiagaraDataInterfaceSkeletalMesh_0}
   }
  }
 ]
}
